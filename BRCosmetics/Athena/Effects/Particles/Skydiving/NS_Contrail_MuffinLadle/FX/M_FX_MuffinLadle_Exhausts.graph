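# Translucent unlit Niagara mesh-particle material for exhaust streaks in the MuffinLadle skydiving contrail.
# A brightness texture blends edge and core colors, which are tinted along V, given a sphere-mask glow and hue-shifted by dynamic parameter (both lerp ends read HueShiftStart).
# Opacity combines particle alpha, brightness plus glow, near-camera fade and optional end fade.
# World position offset combines SplineThicken with an optional curved bend driven by a second dynamic parameter.
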Material /BRCosmetics/Athena/Effects/Particles/Skydiving/NS_Contrail_MuffinLadle/FX/M_FX_MuffinLadle_Exhausts
graph {
 properties {
  bAutomaticallySetUsageInEditor=false
  bUsedWithNiagaraMeshParticles=true
  blendMode=BLEND_Translucent
  shadingModel=MSM_Unlit
  usedShadingModels=Unlit
 }
 nodes [
  Constant() {r=250}
  n1 = DynamicParameter() {defaultValue=<rgba 1 1 0 0>, paramNames [Param1, Param2, HueShiftStart, HueShiftEnd]}
  n2 = ComponentMask(None=TextureCoordinate()) {g=true}
  n3 = VectorParameter() {parameterName=EndTint, defaultValue=<rgba 1 1 1 1>}
  n4 = VectorParameter() {parameterName=StartTint, defaultValue=<rgba 1 1 1 1>}
  n5 = VectorParameter() {parameterName=EdgeColor, defaultValue=<rgba 0 0 0 1>}
  n6 = VectorParameter() {parameterName=CoreColor, defaultValue=<rgba 0 0 0 1>}
  n7 = LinearInterpolate(
   A=TextureCoordinate()
   B=OneMinus(None=MaterialFunctionCall() {materialFunction=&Swizzle}.YX)
   Alpha=ScalarParameter() {defaultValue=1, parameterName=SwizzleUVs}
  )
  n8 = ComponentMask(None=VectorParameter() {group=TextureTransform, parameterName=TextureScale, defaultValue=<rgba 1 1 0 1>}.RGB) {g=true, r=true}
  n9 = VectorParameter() {group=TextureTransform, parameterName=TextureTranslate, defaultValue=<rgba 0 0 0 1>}
  n10 = TextureSampleParameter2D(UVs=Add(A=Multiply(A=@n7, B=@n8), B=ComponentMask(None=@n9.RGB) {g=true, r=true})) {
   parameterName=T_BrightnessMap
   samplerSource=SSM_Clamp_WorldGroupSettings
   samplerType=SAMPLERTYPE_LinearColor
   texture=/BRCosmetics/Effects/Fort_Effects/Textures/Fire/Candle
  }
  n11 = Multiply(
   A=Multiply(A=LinearInterpolate(A=@n3.RGB, B=@n4.RGB, Alpha=@n2), B=LinearInterpolate(A=@n5.RGB, B=@n6.RGB, Alpha=@n10.G))
   B=ParticleColor().RGB
  )
  n12 = SphereMask(A=ComponentMask(None=TextureCoordinate()) {r=true}, B=Constant() {r=0.5}) {attenuationRadius=0.5, hardnessPercent=0}
  n13 = TextureCoordinate()
  n14 = Multiply(
   A=Multiply(A=Power(Base=@n12), B=ScalarParameter() {defaultValue=0.5, parameterName=GlowOpacity})
   B=SphereMask(A=ComponentMask(None=@n13) {g=true}, B=Constant() {r=0.5}) {attenuationRadius=0.5, hardnessPercent=0}
  )
  n15 = Multiply(A=@n14, B=VectorParameter() {parameterName=GlowColor, defaultValue=<rgba 0 0 0 1>}.RGB)
  MaterialFunctionCall(Emissive=MaterialFunctionCall(Hue Shift Percentage=LinearInterpolate(A=@n1.HueShiftStart, B=@n1.HueShiftStart, Alpha=@n2), Texture=Add(A=@n11, B=@n15)) {materialFunction=&HueShift}.Result) {materialFunction=&MF_FX_EmissiveAdjustments}
  MaterialFunctionCall() {materialFunction=&ObjectPivotPoint}
  n16 = ParticleColor()
  n17 = Saturate(None=Add(A=@n10.G, B=@n14))
  n18 = SphereMask(
   A=WorldPosition().XYZ
   B=CameraPositionWS()
   Radius=ScalarParameter() {defaultValue=100, group="6.1) Opacity", parameterName=Camera Distance Falloff, sortPriority=4}
   Hardness=ScalarParameter() {defaultValue=0.2, group="6.1) Opacity", parameterName=Camera Distance Falloff Hardness}
  )
  n19 = TextureCoordinate()
  Multiply(
   A=Multiply(A=Multiply(A=@n16.A, B=@n17), B=OneMinus(None=@n18))
   B=LinearInterpolate(B=ComponentMask(None=@n19) {g=true}, Alpha=ScalarParameter() {parameterName=UseEndFade}) {constA=1}
  )
  n20 = DynamicParameter() {
   parameterIndex=1
   defaultValue=<rgba 1 1 0 0>
   paramNames [??/BendTargetX, ??/BendTargetY, HueShiftStart/BendTargetZ, HueShiftEnd/BendWeight]
  }
  n21 = MaterialFunctionCall(R=@n20.??/BendTargetX, G=@n20.??/BendTargetX, B=@n20.??/BendTargetX) {materialFunction=&AppendMany}
  n22 = NamedRerouteDeclaration(None=MaterialFunctionCall() {materialFunction=&ObjectPivotPoint}.Mesh Particle Pivot Location) {name=Pivot, nodeColor=<rgba 0 0.4375 1 1>}
  n23 = Subtract(A=@n21.RGB, B=@n22)
  Normalize(VectorInput=@n23)
  n24 = ScalarParameter() {group=WPO, parameterName=UseBends}
  n25 = Normalize(VectorInput=Transform(None=Constant3Vector() {constant=<rgba 1 0 0 0>}) {transformSourceType=TRANSFORMSOURCE_Instance})
  n26 = Distance(A=@n21.RGB, B=@n22)
  n27 = TextureCoordinate()
  n28 = Power(Base=OneMinus(None=ComponentMask(None=@n27) {g=true}), Exp=ScalarParameter() {defaultValue=2, group=WPO, parameterName=BendShape})
  n29 = Multiply(A=LinearInterpolate(A=Multiply(A=@n25, B=@n26), B=@n23, Alpha=@n28), B=@n28)
  n30 = ScalarParameter() {parameterName=ThicknessStart}
  n31 = ScalarParameter() {parameterName=ThicknessEnd}
  n32 = StaticBoolParameter() {defaultValue=true, parameterName=UorVExpansion}
  Reroute(None=Add(
    A=MaterialFunctionCall(WidthBase=@n30, WidthTip=@n31, Expand U or V UV Channel=@n32) {materialFunction=&SplineThicken}.WorldPositionOffset
    B=Multiply(A=@n24, B=@n29)
   ))
  Transform(None=Constant3Vector() {constant=<rgba 0 1 0 0>}) {transformSourceType=TRANSFORMSOURCE_Instance}
 ]
}
