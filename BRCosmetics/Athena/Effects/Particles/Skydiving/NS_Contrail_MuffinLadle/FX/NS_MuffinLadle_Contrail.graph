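# Skydiving contrail cosmetic (MuffinLadle): small mesh ships escort the player with exhaust, glow sprites and wind effects.
# BankingShips spawns three particle pairs that ease into skeleton-relative offsets, then orient by velocity, dive and bank angle, with front-end and capture overrides.
# Two smoothed orientations drive a lagging exhaust target, passed to materials through dynamic parameters for bending.
# GPU speed-line sprites scale spawn rate with dive; their duplicate emitter is disabled.
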NiagaraSystem /BRCosmetics/Athena/Effects/Particles/Skydiving/NS_Contrail_MuffinLadle/FX/NS_MuffinLadle_Contrail
dependencies {
 usedDataInterfaces [
  &NiagaraDataInterfaceCurve
  &NiagaraDataInterfaceVector2DCurve
  &NiagaraDataInterfaceArrayFloat
  &NiagaraDataInterfaceSkeletalMesh
  &NiagaraDataInterfaceArrayFloat3
 ]
 usedDynamicInputs [
  &Vector2DFromCurve
  &OneMinusFloat
  &Divide_Float
  /Niagara/DynamicInputs/Lerp/SmoothLerpOverTime_Float
 ]
 usedModules [
  &self:SystemSpawnScript.SystemScriptSource.SystemScriptGraph.NiagaraNodeAssignment_11.SetVariables_#1
  &self:SystemSpawnScript.SystemScriptSource.SystemScriptGraph.NiagaraNodeAssignment_5.SetVariables_#2
  &self:SystemSpawnScript.SystemScriptSource.SystemScriptGraph.NiagaraNodeAssignment_0.SetVariables_#3
  &self:SystemSpawnScript.SystemScriptSource.SystemScriptGraph.NiagaraNodeAssignment_2.SetVariables_#4
  &self:SystemSpawnScript.SystemScriptSource.SystemScriptGraph.NiagaraNodeAssignment_3.SetVariables_#5
  &self:SystemSpawnScript.SystemScriptSource.SystemScriptGraph.NiagaraNodeAssignment_6.SetVariables_#6
  &self:SystemSpawnScript.SystemScriptSource.SystemScriptGraph.NiagaraNodeAssignment_12.SetVariables_#7
  &self:SystemSpawnScript.SystemScriptSource.SystemScriptGraph.NiagaraNodeAssignment_14.SetVariables_#8
  &self:SystemSpawnScript.SystemScriptSource.SystemScriptGraph.NiagaraNodeAssignment_13.SetVariables_#9
  &SystemState
  /Game/Effects/Niagara/Modules/Contrails/FN_Contrail_Setup
  /Game/Effects/Niagara/Modules/Skydiving/FN_DivingAttributes/v6/FN_DivingAttributes
  &self:ScratchModule
  &self:SystemSpawnScript.SystemScriptSource.SystemScriptGraph.NiagaraNodeAssignment_4.SetVariables_#10
  &self:SystemSpawnScript.SystemScriptSource.SystemScriptGraph.NiagaraNodeAssignment_1.SetVariables_#11
  &FN_Integrator
  &self:SystemSpawnScript.SystemScriptSource.SystemScriptGraph.NiagaraNodeAssignment_10.SetVariables_#12
  &Timeline
  &self:SystemSpawnScript.SystemScriptSource.SystemScriptGraph.NiagaraNodeAssignment_9.SetVariables_#13
  &self:SystemSpawnScript.SystemScriptSource.SystemScriptGraph.NiagaraNodeAssignment_7.SetVariables_#14
  &self:SystemSpawnScript.SystemScriptSource.SystemScriptGraph.NiagaraNodeAssignment_8.SetVariables_#15
 ]
}
emitters {
 BankingShips {
  emitterSpawnScript [/SetVariables(Emitter.BendWeight=1) {isSetParameters=true}]
  emitterUpdateScript [
   &EmitterState()
   &SpawnBurst_Instantaneous(Age:NiagaraFloat=$Emitter.Age, Spawn Count=2, Use Spawn Probability=false, UseLoopCountLimit=false)
   &SpawnBurst_Instantaneous(Age:NiagaraFloat=$Emitter.Age, Spawn Count=2, Spawn Group=1, Use Spawn Probability=false, UseLoopCountLimit=false) {module=SpawnBurst_Instantaneous002}
   &SpawnBurst_Instantaneous(Age:NiagaraFloat=$Emitter.Age, Spawn Count=2, Spawn Group=2, Use Spawn Probability=false, UseLoopCountLimit=false) {module=SpawnBurst_Instantaneous001}
   &InterpolateOverTime(Rate Of Change=50, Target Vector Value:Vector3f=$Engine.Owner.Velocity, Variable Type=::NewEnumerator2)
   /SetVariables(Emitter.ExhaustBlend=1) {isSetParameters=true}
   /SetVariables(
    Emitter.WindCamOffset:NiagaraFloat=&FN_FrontEndFloatModifier(Gameplay=0, Front End=10)
    Emitter.WindWeight:NiagaraFloat=&FN_FrontEndFloatModifier(
     Gameplay=$System.FN_Contrail_Setup.Dive Value
     Front End=&Lerp_Float(A=0.75, B=1, Alpha=$System.FN_Contrail_Setup.Dive Value)
    )
   ) {isSetParameters=true}
   /SetVariables(Emitter.WindParams:Vector4f=&MakeVector4(X=1, Y=2, Z=0, W=0)) {isSetParameters=true}
   /SetVariables() {isSetParameters=true}
   &SpawnBurst_Instantaneous(Spawn Count=2, Spawn Time=1.25, Use Spawn Probability=false, UseLoopCountLimit=false) {enabled=false, module=SpawnBurst_Instantaneous003}
  ]
  particleSpawnScript [
   &PartitionParticles(Particles Per Partition=1, Partition Mode:ENiagara_PartitionMode_v2=::NewEnumerator1)
   &InitializeParticle(
    Color Mode=::NewEnumerator0
    Lifetime=5
    Mesh Scale Mode=::NewEnumerator3
    Mesh Uniform Scale=$System.Ship Scale
    Position Offset=<0 0 1000>
    Sprite Size Mode=::NewEnumerator3
    Uniform Sprite Size=15
    UsePositionOffset=true
    Write Material Random=true
   )
   &ShapeLocation(MaskSpawnGroup=false, Override Seed=false, UseOwnerScale=true)
   &InitialMeshOrientation(Mesh Orientation Mode=::NewEnumerator7, Use Rotation Vector=true)
   /SetVariables(Particles.Debug sprite offset=50) {isSetParameters=true}
   /SetVariables(Particles.ShipIndex:NiagaraInt32=&MergeHelper_IntSwitch(0=0, 1=1, 2=1, 3=?, 4=?, 5=?, Index=$Emitter.SpawnGroup, Num=::NewEnumerator1)) {isSetParameters=true}
   /SetVariables(Particles.ShipNum:NiagaraInt32=$Emitter.SpawnGroup) {isSetParameters=true}
   /SetVariables(Particles.ShipVizTag:NiagaraInt32=$Particles.PartitionParticles.Partition) {isSetParameters=true}
   /SetVariables(Particles.LagPosTarget:NiagaraPosition=/Niagara/DynamicInputs/Helpers/SimulationPosition()) {isSetParameters=true}
  ]
  particleUpdateScript [
   &ParticleState(Kill Particles When Lifetime Has Elapsed=false, Let Infinitely Lived Particles Die When Emitter Deactivates=true)
   /SetVariables(Particles.OverallShipOffset:Vector3f=&MakeCustomVectorFromBool(
     Bool=&Boolean_OrOperation(Bool1=$System.Capture, Bool2=false)
     True Vector=<-20.6248 -3.5921 3.9304>
     False Vector=&FN_FrontEndVector3Modifier(
      Gameplay=&Lerp_Vector(
       A=<0 0 500>
       B=&Lerp_Vector(A=<0 0 10>, B=<0 50 50>, Alpha=$System.FN_Contrail_Setup.Dive Value)
       Alpha=&FloatFromCurve(
        FloatCurve=DataInterface() {propertyValues {Curve {keys [
            {arriveTangent=1.2212, interpMode=RCIM_Cubic, leaveTangent=1.2212, tangentMode=RCTM_User, time=0, value=0}
            {arriveTangent=0.1897, interpMode=RCIM_Cubic, leaveTangent=0.1897, tangentMode=RCTM_User, time=1.4, value=0.975}
            {arriveTangent=0.0096, interpMode=RCIM_Cubic, leaveTangent=0.0096, time=3, value=0.95}
            {arriveTangent=0.0257, interpMode=RCIM_Cubic, leaveTangent=0.0257, tangentMode=RCTM_User, time=4, value=1}
           ]}}}
        CurveIndex=&MergeHelper_FloatSwitch(
         0=&Multiply_Float(A=$Engine.System.Age, B=1.05)
         1=&Multiply_Float(A=$Engine.System.Age, B=0.7)
         2=?
         3=?
         Index=$Particles.ShipIndex
         Num=::NewEnumerator0
        )
       )
      )
      Front End=&Lerp_Vector(
       A=<0 0 500>
       B=&Lerp_Vector(A=<0 -40 25>, B=<0 -50 10>, Alpha=$System.FN_Contrail_Setup.Dive Value)
       Alpha=&FloatFromCurve(
        FloatCurve=DataInterface() {propertyValues {Curve {keys [
            {arriveTangent=1.2212, interpMode=RCIM_Cubic, leaveTangent=1.2212, tangentMode=RCTM_User, time=0, value=0}
            {arriveTangent=0.1897, interpMode=RCIM_Cubic, leaveTangent=0.1897, tangentMode=RCTM_User, time=1.4, value=0.975}
            {arriveTangent=0.0096, interpMode=RCIM_Cubic, leaveTangent=0.0096, time=3, value=0.95}
            {arriveTangent=0.0257, interpMode=RCIM_Cubic, leaveTangent=0.0257, tangentMode=RCTM_User, time=4, value=1}
           ]}}}
        CurveIndex=&MergeHelper_FloatSwitch(
         0=&Multiply_Float(A=$Engine.System.Age, B=1.05)
         1=&Multiply_Float(A=$Engine.System.Age, B=0.7)
         2=?
         3=?
         Index=$Particles.ShipIndex
         Num=::NewEnumerator0
        )
       )
      )
     )
    )) {isSetParameters=true}
   /SetVariables(Particles.Ambient:Vector3f=&Multiply_VectorByFloat(
     Vector=&MakeVector(
      X=&Waveform(
       "Waveform [1]"=::NewEnumerator12
       "[1] Input Position"=&SelectFloatFromArray(Direct Array Index=$Particles.ShipNum, Float Selection Array=DataInterface() {propertyValues {FloatData [200,1000]}})
       "[1] Frequency"=1
       "[1] Amplitude Min/Max"=<-50 50>
      )
      Y=&Waveform(
       "Waveform [1]"=::NewEnumerator12
       "[1] Input Position"=&SelectFloatFromArray(Direct Array Index=$Particles.ShipNum, Float Selection Array=DataInterface() {propertyValues {FloatData [1200,1800]}})
       "[1] Frequency"=1
       "[1] Amplitude Min/Max"=<-50 50>
      )
      Z=&Waveform(
       "Waveform [1]"=::NewEnumerator12
       "[1] Input Position"=&SelectFloatFromArray(Direct Array Index=$Particles.ShipNum, Float Selection Array=DataInterface() {propertyValues {FloatData [2200,2800]}})
       "[1] Frequency"=1
       "[1] Amplitude Min/Max"=<-50 50>
      )
     )
     Float=&Lerp_Float(A=0, B=0, Alpha=$System.FN_Contrail_Setup.Dive Value)
    )) {isSetParameters=true}
   &InterpolateOverTime(
    Rate Of Change=3
    Target Float Value:NiagaraFloat=&MakeCustomFloatFromBool(Bool=$System.Timeline001.IsRewinding, True Float=0, False Float=1)
   ) {module=InterpolateOverTime004}
   /SetVariables(Particles.CatchupVector:Vector3f=&Multiply_VectorByFloat(
     Vector=&Multiply_VectorByFloat(
      Vector=&NormalizeVector(Vector To Normalize=$System.SkydiveVelocity, Safe Normalize=true, Safe Fallback Threshold=0, Safe Fallback Vector=<0 0 0.001>)
      Float=-75
     )
     Float=&Lerp_Float(
      A=&FloatFromCurve(
       FloatCurve=DataInterface() {propertyValues {Curve {keys [
           {interpMode=RCIM_Cubic, tangentMode=RCTM_User, time=0, value=0}
           {interpMode=RCIM_Cubic, tangentMode=RCTM_User, time=0.35, value=-0.35}
           {tangentMode=RCTM_User, time=1, value=0}
          ]}}}
       CurveIndex=$System.Timeline001.NormalizedTime
      )
      B=&FloatFromCurve(
       FloatCurve=DataInterface() {propertyValues {Curve {keys [
           {arriveTangent=3.3827, arriveTangentWeight=1.1106, interpMode=RCIM_Cubic, leaveTangentWeight=1.1106, time=0, value=0}
           {arriveTangentWeight=0.5241, interpMode=RCIM_Cubic, leaveTangentWeight=0.1667, time=0.5, value=0.65}
           <at 1 0 RCIM_Cubic>
          ]}}}
       CurveIndex=$System.Timeline001.NormalizedTime
       Scale Curve=0.1
      )
      Alpha=$Particles.InterpolateOverTime004.InterpolateOverTime_Float.Moving Average
     )
    )) {isSetParameters=true}
   &SkeletalMeshLocation(
    Bone Sampling Mode=::NewEnumerator4
    MaskSpawnGroup=false
    Orientation Sampling=::NewEnumerator1
    Sampled Position Offset=&Add_Vector(
     A=$Particles.CatchupVector
     B=/Niagara/DynamicInputs/Vectors/RotateVector(
      Vector To Rotate:Vector3f=&Add_Vector(
       A=$Particles.Ambient
       B=&Add_Vector(
        A=&SelectVectorFromArray(Direct Array Index=$Particles.ShipNum, Direct Array Mode=::NewEnumerator1, Vector Selection Array=$System.ShipPositions)
        B=$Particles.OverallShipOffset
       )
      )
      Yaw:NiagaraFloat=&Waveform(
       "Waveform [1]"=::NewEnumerator12
       "[1] Input Position"=&Multiply_FloatByInt(Float=200000, Integer=$Particles.ShipIndex)
       "[1] Frequency"=1.25
       "[1] Amplitude Min/Max"=<7 -7>
       "[1] Scale"=&MakeCustomFloatFromBool(Bool=$System.Capture, True Float=0, False Float=1)
      )
      Pitch:NiagaraFloat=0
      Roll:NiagaraFloat=0
      Rotation Rate:NiagaraFloat=1
     )
    )
    Skeletal Mesh=$System.SkeletalMesh
    Use Sampled Mesh Attribute as Ribbon ID=false
   )
   &ScaleMeshSize(Scale Factor=&VectorFromFloat(Value=&MakeCustomFloatFromBool(
      Bool=$System.Capture
      True Float=1
      False Float=&FloatFromCurve(
       FloatCurve=DataInterface() {propertyValues {Curve {keys [{arriveTangent=1.0186, interpMode=RCIM_Cubic, leaveTangent=1.0186, tangentMode=RCTM_User, time=0, value=0}, <at 1 1 RCIM_Cubic>]}}}
       CurveIndex=$Particles.Age
      )
     )))
   &Drag(Drag=2, Use Linear Drag=true, Use Rotational Drag=false)
   &SolveForcesAndVelocity(Clamp Velocity=false, Limit Acceleration=false)
   &UpdateMeshOrientation(
    Bank Rate:NiagaraFloat=&Lerp_Float(A=1, B=900, Alpha=$System.FN_Contrail_Setup.Dive Value)
    Banking Angle Min / Max:Vector2f=&Lerp_Vector2(A=<0 0>, B=<0 0>, Alpha=$System.FN_Contrail_Setup.Dive Value)
    "Foward Vector (Mesh Space)":Vector3f=&Lerp_Vector(A=<1 0 1>, B=<1 0 0>, Alpha=$System.FN_Contrail_Setup.Dive Value)
    Orientation Method=::NewEnumerator3
    Particle Velocity:Vector3f=$Emitter.InterpolateOverTime.InterpolateOverTime_V3_.Moving Average
    "Right Vector (Mesh Space)":Vector3f=&Lerp_Vector(
     A=<0 1 -0.015>
     B=<0 1 0.015>
     Alpha=&MakeFloatFromVector(Vector=$Particles.Ambient, Channel=::NewEnumerator0)
    )
    Rotation Decay Rate In Degrees:NiagaraFloat=&Lerp_Float(A=150, B=500, Alpha=$System.Dive)
    Up Vector:Vector3f=&Lerp_Vector(
     A=&ReturnMeshOrientationAxis(Mesh Orientation Axis=::YAxis, Mesh Orientation Quat=$Particles.SkeletalMeshLocation.SampledRotation)
     B=&ReturnMeshOrientationAxis(Mesh Orientation Axis=::ZAxis, Mesh Orientation Quat=$Particles.SkeletalMeshLocation.SampledRotation)
     Alpha=&FloatFromCurve(FloatCurve=DataInterface() {propertyValues {Curve {keys [<at 0 -0.25>,<at 0.25 0.5>]}}}, CurveIndex=$System.FN_Contrail_Setup.Dive Value)
    )
    Upright Roll Correction Speed:NiagaraFloat=&Lerp_Float(A=25, B=90000, Alpha=$System.Dive)
    Velocity Orientation Convergence Rate:NiagaraFloat=&Lerp_Float(
     A=&Lerp_Float(
      A=3
      B=12
      Alpha=&Waveform(
       "Waveform [1]"=::NewEnumerator12
       "[1] Input Position"=&RandomRangeFloat(Minimum=0, Maximum=100)
       "[1] Frequency"=5
       "[1] Amplitude Min/Max"=<0 1>
      )
     )
     B=25
     Alpha=$System.Dive
    )
   ) {module=UpdateMeshOrientation001}
   /SetVariables(Particles.FrontendRotation:Quat4f=&MultiplyQuaternion(
     Quaternion A=/CascadeToNiagaraConverter/NiagaraScripts/DynamicInputScripts/MakeQuatFromEuler(Euler:Vector3f=<0 0 90>, Scale:NiagaraFloat=1)
     Quaternion B=$Particles.SkeletalMeshLocation.SampledRotation
    )) {isSetParameters=true}
   /SetVariables(Particles.FrontendRotation:Quat4f=&MultiplyQuaternion(
     Quaternion A=$StackContext.FrontendRotation
     Quaternion B=&MakeQuaternion(
      Quaternion From=0
      Axis=<0 1 0>
      Angle=&Lerp_Float(A=25, B=45, Alpha=$Emitter.WindWeight)
      From=?
      To=?
      Angle Type=::NewEnumerator0
      Coordinate Space=::World
      Normalize Vectors=false
     )
    )) {isSetParameters=true}
   /SetVariables(Particles.BankOrientation:Quat4f=/Niagara/DynamicInputs/TypeConversions/MakeCustomQuaternionFromBool(
     Bool:NiagaraBool=$User.bIsFrontEndPreview
     True Quat:Quat4f=$StackContext.FrontendRotation
     False Quat:Quat4f=&MakeCustomQuaternionFromBool(
      Bool:NiagaraBool=$System.Capture
      True Quat:Quat4f=&MakeQuatFromEuler(Euler:Vector3f=<0 -35 90>, Scale:NiagaraFloat=1)
      False Quat:Quat4f=&MultiplyQuaternion(
       Quaternion A=$Particles.MeshOrientation
       Quaternion B=&MakeQuaternion(
        Quaternion From=0
        Axis=<1 0 0>
        Angle=$System.BankAngle
        From=?
        To=?
        Angle Type=::NewEnumerator0
        Coordinate Space=::World
        Normalize Vectors=false
       )
      )
     )
    )) {isSetParameters=true}
   &InterpolateOverTime(Rate Of Change=5, Target Quat Value:Quat4f=$Particles.BankOrientation, Variable Type=::NewEnumerator5) {module=InterpolateOverTime002}
   /SetVariables(Particles.New Orient:Quat4f=$Particles.InterpolateOverTime002.InterpolateOverTime_Quaternion.Moving Average) {isSetParameters=true}
   &InterpolateOverTime(Rate Of Change=4, Target Quat Value:Quat4f=$Particles.BankOrientation, Variable Type=::NewEnumerator5) {module=InterpolateOverTime001}
   /SetVariables(Particles.LazierQuat:Quat4f=$StackContext.InterpolateOverTime001.InterpolateOverTime_Quaternion.Moving Average) {isSetParameters=true}
   /SetVariables(Particles.MeshOrientation=$Particles.New Orient) {isSetParameters=true}
   /SetVariables(Particles.ExhaustPosition:NiagaraPosition=/Niagara/DynamicInputs/Math/TransformVectorByMatrix(
     Vector:Vector3f=&MergeHelper_VectorSwitch(
      0=&MergeHelper_VectorSwitch(0=<-25 -5 0>, 1=<-25 5 0>, 2=?, 3=?, Index=$Particles.PartitionParticles.Partition, Num=::NewEnumerator0)
      1=&MergeHelper_VectorSwitch(0=<-10 8 0>, 1=<-10 -8 0>, 2=?, 3=?, Index=$Particles.PartitionParticles.Partition, Num=::NewEnumerator0)
      2=?
      3=?
      Index=$Particles.ShipIndex
      Num=::NewEnumerator0
     )
     Transformation Matrix:NiagaraMatrix=/Niagara/DynamicInputs/Math/MakeTransform(
      Inverse:NiagaraBool=false
      Position:Vector3f=Linked() {linkedVariable {name=Particles.Position, type=NiagaraPosition}}
      Local Offset:Vector3f=<0 0 0>
      Scale:Vector3f=$Particles.Scale
      Rotation Interface:ENiagara_RotationDefinitionApproach=::NewEnumerator1
      Axis:Vector3f=?
      Quaternion:Quat4f=$Particles.MeshOrientation
      Rotation Matrix:NiagaraMatrix=?
      X Vector:Vector3f=?
      Yaw:NiagaraFloat=?
      Angle:NiagaraFloat=?
      Pitch:NiagaraFloat=?
      Y Vector:Vector3f=?
      Roll:NiagaraFloat=?
      Z Vector:Vector3f=?
     )
     Transform Vector as Position:NiagaraBool=true
    )) {isSetParameters=true}
   /SetVariables(Particles.LagPosTarget:NiagaraPosition=/Niagara/DynamicInputs/Lerp/SmoothLerpOverTime_Position(
     Convergence Rate:NiagaraFloat=100
     Smooth Value:NiagaraPosition=$Particles.LagPosTarget
     Target Value:NiagaraPosition=&TransformVectorByMatrix(
      Vector:Vector3f=&MakeVector(X=&Lerp_Float(A=-5, B=-85, Alpha=$System.FN_Contrail_Setup.Dive Value), Y=0, Z=0)
      Transformation Matrix:NiagaraMatrix=&MakeTransform(
       Inverse:NiagaraBool=false
       Position:Vector3f=Linked() {linkedVariable {name=Particles.ExhaustPosition, type=NiagaraPosition}}
       Local Offset:Vector3f=<0 0 0>
       Scale:Vector3f=$Particles.Scale
       Rotation Interface:ENiagara_RotationDefinitionApproach=::NewEnumerator1
       Axis:Vector3f=?
       Quaternion:Quat4f=$Particles.LazierQuat
       Rotation Matrix:NiagaraMatrix=?
       X Vector:Vector3f=?
       Yaw:NiagaraFloat=?
       Angle:NiagaraFloat=?
       Pitch:NiagaraFloat=?
       Y Vector:Vector3f=?
       Roll:NiagaraFloat=?
       Z Vector:Vector3f=?
      )
      Transform Vector as Position:NiagaraBool=true
     )
    )) {isSetParameters=true}
   /SetVariables(Particles.LagDirection:Vector3f=&NormalizeVector(
     Vector To Normalize=&SubtractVectorFromPosition(Position=$Particles.LagPosTarget, Vector=Linked() {linkedVariable {name=Particles.TailEndPos, type=NiagaraPosition}})
     Safe Normalize=true
     Safe Fallback Threshold=0
     Safe Fallback Vector=<0 0 0.001>
    )) {isSetParameters=true}
   &DynamicMaterialParameters(
    Index 0 Param 1=1
    Index 0 Param 3=&MergeHelper_FloatSwitch(0=-0.05, 1=0.05, 2=?, 3=?, Index=$Particles.ShipIndex, Num=::NewEnumerator0)
    Index 0 Param 4=&MergeHelper_FloatSwitch(0=0, 1=0, 2=?, 3=?, Index=$Particles.ShipIndex, Num=::NewEnumerator0)
    Index 1 Param 1=&MakeFloatFromVector(Vector=$Particles.LagPosTarget, Channel=::NewEnumerator0)
    Index 1 Param 2=&MakeFloatFromVector(Vector=$Particles.LagPosTarget, Channel=::NewEnumerator1)
    Index 1 Param 3=&MakeFloatFromVector(Vector=$Particles.LagPosTarget, Channel=::NewEnumerator2)
    Index 1 Param 4=&ClampFloat(Float=&RemapRange(
      Input Value To Remap=&self:AngleBetweenQuats(Quat:Quat4f=$Particles.MeshOrientation, Quat B:Quat4f=$StackContext.LazierQuat)
      Input Min=0
      Input Max=0.1
      Output Min=0
      Output Max=1
     ))
    Index 2 Param 1=&MakeFloatFromVector(
     Vector=&RotateVectorByQuaternion(Vector To Rotate=<1 0 0>, Quaternion=$Particles.MeshOrientation)
     Channel=::NewEnumerator0
    )
    Index 2 Param 2=&MakeFloatFromVector(
     Vector=&RotateVectorByQuaternion(Vector To Rotate=<1 0 0>, Quaternion=$Particles.MeshOrientation)
     Channel=::NewEnumerator1
    )
    Index 2 Param 3=&MakeFloatFromVector(
     Vector=&RotateVectorByQuaternion(Vector To Rotate=<1 0 0>, Quaternion=$Particles.MeshOrientation)
     Channel=::NewEnumerator2
    )
    Module.Param0WriteEnabled=true
    Module.Param0WriteEnabled1=true
    Module.Param0WriteEnabled2=true
    Module.Param1WriteEnabled=true
    Module.Param1WriteEnabled1=true
    Module.Param1WriteEnabled2=true
    Module.Param2WriteEnabled=true
    Module.Param2WriteEnabled1=true
    Module.Param2WriteEnabled2=true
    Module.Param3WriteEnabled=true
    Module.Param3WriteEnabled1=true
    Module.Param3WriteEnabled2=true
    Write Parameter Index 1=true
    Write Parameter Index 2=true
   )
   &FN_Mesh_Attributes(
    Color=&MultiplyLinearColor(
     Linear Color=<rgba 1 1 1 1>
     Scale Factor=&MakeVector4FromVectorAndFloat(
      "Vector (RGB)"=<1 1 1>
      "Float (Alpha)"=&FN_FrontEndFloatModifier(Gameplay=$Emitter.WindWeight, Front End=&Lerp_Float(A=0.75, B=1.4, Alpha=$System.Dive))
     )
    )
    Create Color Attribute=true
    Create Scale Attribute=true
    Scale=&Multiply_Vector(
     A=&MergeHelper_VectorSwitch(0=<0.8 1 1>, 1=<1 1 1>, 2=?, 3=?, Index=$Particles.ShipIndex, Num=::NewEnumerator0)
     B=$Particles.Scale
    )
    Write to Module Namespace=true
   )
   &FN_Mesh_Attributes(
    Color=&MultiplyLinearColor(
     Linear Color=&MultiplyLinearColor(
      Linear Color=&MergeHelper_ColorSwitch(0=$System.MainExhaustColor, 1=$System.SecondExhaustColor, 2=?, 3=?, Index=$Particles.ShipIndex, Num=::NewEnumerator0)
      Scale Factor=&MakeVector4FromVectorAndFloat(
       "Vector (RGB)"=&VectorFromFloat(Value=&Waveform(
         Global Input Position=&RandomRangeFloat(Minimum=0, Maximum=1)
         "Waveform [1]"=::NewEnumerator12
         "[1] Frequency"=3
         "[1] Amplitude Min/Max"=<0.9 1.5>
        ))
       "Float (Alpha)"=1
      )
     )
     Scale Factor=&MakeVector4(
      X=1
      Y=1
      Z=1
      W=&FloatFromCurve(FloatCurve=DataInterface() {propertyValues {Curve {keys [<at 0 1>,<at 1 0>]}}}, CurveIndex=$System.IsIdle)
     )
    )
    Create Color Attribute=true
    Create Scale Attribute=true
    Scale=&Lerp_Vector(
     A=&Multiply_Vector(
      A=&Multiply_Vector(
       A=&MakeVector(
        X=&Multiply_Float(
         A=&Lerp_Float(A=0, B=1, Alpha=$Emitter.ExhaustBlend)
         B=&Waveform(
          Global Input Position=&RandomRangeFloat(Minimum=0, Maximum=1)
          "Waveform [1]"=::NewEnumerator12
          "[1] Frequency"=12
          "[1] Amplitude Min/Max"=<1 1>
         )
        )
        Y=1
        Z=1
       )
       B=&Lerp_Vector(A=<0.6 1 1>, B=<1 1 1>, Alpha=$System.Dive)
      )
      B=$StackContext.Scale
     )
     B=<1 1 1>
     Alpha=1
    )
    Write to Module Namespace=true
   ) {module=FN_Mesh_Attributes001}
   &FN_Sprite_Attributes(
    Camera Offset=5
    Color=&MultiplyLinearColor(
     Linear Color=&MergeHelper_ColorSwitch(0=$System.MainExhaustColor, 1=$System.SecondExhaustColor, 2=?, 3=?, Index=$Particles.ShipIndex, Num=::NewEnumerator0)
     Scale Factor=&MakeVector4(
      X=1
      Y=1
      Z=1
      W=&FloatFromCurve(FloatCurve=DataInterface() {propertyValues {Curve {keys [<at 0 1>,<at 1 0>]}}}, CurveIndex=$System.IsIdle)
     )
    )
    Create Camera Offset Attribute=true
    Create Color Attribute=true
    Create Sprite Rotation Attribute=true
    Create SpriteSize Attribute=true
    SpriteRotation:NiagaraFloat=&RandomRangeFloat(Minimum=0, Maximum=1)
    SpriteSize=&Multiply_Vector2D_ByFloat(
     Float=&Waveform(
      Global Input Position=&RandomRangeFloat(Minimum=0, Maximum=1)
      "Waveform [1]"=::NewEnumerator12
      "[1] Frequency"=&Lerp_Float(A=8, B=12, Alpha=$System.FN_Contrail_Setup.Dive Value)
      "[1] Amplitude Min/Max"=<0.75 1.2>
     )
     Vector2D=<20 20>
    )
    Write to Module Namespace=true
   )
   &FN_Mesh_Attributes(
    Color=&MultiplyLinearColor(
     Linear Color=<rgba 1 1 1 1>
     Scale Factor=&MakeVector4FromVectorAndFloat(
      "Vector (RGB)"=<1 1 1>
      "Float (Alpha)"=&FloatFromCurve(
       FloatCurve=DataInterface() {propertyValues {Curve {
          defaultValue=1
          keys [{interpMode=RCIM_Cubic, tangentMode=RCTM_User, time=0.35, value=0}, {interpMode=RCIM_Cubic, tangentMode=RCTM_User, time=1, value=1}]
         }}}
       CurveIndex=$Emitter.WindWeight
       Scale Curve=0.05
      )
     )
    )
    Create Color Attribute=true
    Create Scale Attribute=true
    Scale=&Multiply_Vector(A=$Particles.Scale, B=&Lerp_Vector(A=<1 1 1>, B=<2 1 1>, Alpha=$Emitter.WindWeight))
    Write to Module Namespace=true
   ) {module=FN_Mesh_Attributes002}
  ]
  rendererClasses [&NiagaraMeshRendererProperties, &NiagaraSpriteRendererProperties]
  renderers [
   {rendererClass=&NiagaraMeshRendererProperties, rendererIndex=0}
   {rendererClass=&NiagaraSpriteRendererProperties, rendererIndex=1}
   {rendererClass=&NiagaraMeshRendererProperties, rendererIndex=2}
   {rendererClass=&NiagaraMeshRendererProperties, rendererIndex=3}
   {rendererClass=&NiagaraMeshRendererProperties, rendererIndex=4}
  ]
 }
 FN_SpeedLines {
  simTarget=GPUComputeSim
  emitterUpdateScript [
   &EmitterState(Life Cycle Mode=::NewEnumerator1, Loop Duration=1, UseLoopDelay=false)
   &SpawnRate(
    Spawn Probability=&MakeCustomFloatFromBool(Bool=&SetBoolByFloatComparison(A=$System.Age, B=0.2), True Float=1, False Float=1)
    SpawnRate=&Lerp_Float(
     A=0
     B=&FN_FrontEndFloatModifier(
      Gameplay=&Multiply_Float(A=55, B=$System.FN_Contrail_Setup.Dive Value)
      Front End=&Multiply_Float(A=50, B=&Lerp_Float(A=0.5, B=1, Alpha=$System.FN_Contrail_Setup.Dive Value))
     )
     Alpha=1
    )
    Use Spawn Probability=true
   )
   &SpawnRate(
    Spawn Probability=&MakeCustomFloatFromBool(Bool=&SetBoolByFloatComparison(A=$System.Age, B=0.2), True Float=1, False Float=1)
    SpawnRate=&Lerp_Float(A=0, B=500, Alpha=$System.Dive)
    Use Spawn Probability=true
   ) {enabled=false, module=SpawnRate001}
  ]
  particleSpawnScript [
   &InitializeParticle(
    Color Maximum=<rgba 0.25 0.25 0.25 0.05>
    Color Minimum=<rgba 1 1 1 0.025>
    Color Mode=::NewEnumerator2
    Lifetime Max=0.5
    Lifetime Min=0.3
    Lifetime Mode=::NewEnumerator1
    Sprite Rotation Mode=::NewEnumerator2
    Sprite Size=&UniformRangedVector2D(Minimum=<7 100>, Maximum=<1 200>)
    Sprite Size Mode=::NewEnumerator4
    UsePositionOffset=false
    Write Material Random=true
   ) {module=InitializeParticle001}
   &SkeletalMeshLocation(
    Bone / Socket Index=3
    Bone Sampling Mode=::NewEnumerator4
    MaskSpawnGroup=false
    Sampled Orientation Coordinate Space=::Local
    Skeletal Mesh=$System.SkeletalMesh
   )
   &SphereLocation(
    MaskSpawnGroup=false
    Non Uniform Scale=&RotateVectorByQuaternion(Vector To Rotate=<1 1 1>, Quaternion=$Particles.SkeletalMeshLocation.SampledRotation)
    Offset=&FN_FrontEndVector3Modifier(
     Gameplay=&Add_Vector(
      A=&RotateVectorByQuaternion(Vector To Rotate=<100 0 0>, Quaternion=$Particles.SkeletalMeshLocation.SampledRotation)
      B=&TransformVector(Vector=<0 0 -200>, Source Space=::World, Destination Space=::World)
     )
     Front End=<0 0 0>
    )
    Override Seed=false
    Sphere Orientation Axis:Vector3f=&RotateVectorByQuaternion(Vector To Rotate=<100 0 0>, Quaternion=$Particles.SkeletalMeshLocation.SampledRotation)
    Sphere Origin:NiagaraPosition=$Particles.SkeletalMeshLocation.SampledPosition
    Sphere Radius=250
    Surface Only=false
    UseNonUniformScale=true
    UseOffset=true
    UseSphereOrientation=true
   )
   &InheritVelocity(Module.Limit Speed=true)
   &AddVelocity(
    Velocity=&FN_FrontEndVector3Modifier(Gameplay=<0 0 0>, Front End=<0 -100 100>)
    Velocity Speed Scale=&FN_FrontEndFloatModifier(Gameplay=1, Front End=2)
   )
  ]
  particleUpdateScript [
   &ParticleState()
   &SolveForcesAndVelocity(Clamp Velocity=false, Limit Acceleration=false)
   &ScaleSpriteSize(
    Non-Uniform Curve Scale=&FN_FrontEndVector2Modifier(Front End:Vector2f=<0.5 2>, Gameplay:Vector2f=<1.5 3>)
    Non-Uniform Curve Sprite Scale=DataInterface() {propertyValues {
      XCurve {keys [<at -0.0069 0.5 RCIM_Cubic>,<at 0.4918 1 RCIM_Cubic>,<at 1 0.5 RCIM_Cubic>]}
      YCurve {keys [
        {
         arriveTangent=3.1787
         arriveTangentWeight=0.684
         interpMode=RCIM_Cubic
         leaveTangent=3.1787
         leaveTangentWeight=0.684
         tangentMode=RCTM_User
         tangentWeightMode=RCTWM_WeightedBoth
         time=0
         value=0.25
        }
        {
         arriveTangentWeight=0.2354
         interpMode=RCIM_Cubic
         leaveTangentWeight=0.3527
         tangentMode=RCTM_Break
         tangentWeightMode=RCTWM_WeightedBoth
         time=0.4975
         value=1
        }
        <at 1 0.25 RCIM_Cubic>
       ]}
     }}
    Scale Sprite Size Mode=::NewEnumerator3
   )
   &ScaleColor(
    Scale Alpha=&FloatFromCurve(
     FloatCurve=DataInterface() {propertyValues {Curve {
        defaultValue=1
        keys [
         {arriveTangent=9.0725, interpMode=RCIM_Cubic, leaveTangent=9.0725, tangentMode=RCTM_User, time=0, value=0}
         {arriveTangent=-0.0194, interpMode=RCIM_Cubic, leaveTangent=-0.0194, tangentMode=RCTM_User, time=0.3, value=0.875}
         {arriveTangent=-0.0405, interpMode=RCIM_Cubic, leaveTangent=-0.0405, tangentMode=RCTM_User, time=1, value=0}
        ]
       }}}
     Scale Curve=&FN_FrontEndFloatModifier(Gameplay=0.7, Front End=0.35)
    )
    ScaleA=true
    ScaleRGB=true
   )
  ]
  rendererClasses [&NiagaraSpriteRendererProperties]
  renderers [{rendererClass=&NiagaraSpriteRendererProperties}]
 }
 FN_SpeedLines001 {
  bEnabled=false
  simTarget=GPUComputeSim
  emitterUpdateScript [
   &EmitterState(Life Cycle Mode=::NewEnumerator1, Loop Duration=1, UseLoopDelay=false)
   &SpawnRate(
    Spawn Probability=&MakeCustomFloatFromBool(Bool=&SetBoolByFloatComparison(A=$System.Age, B=0.2), True Float=1, False Float=1)
    SpawnRate=&Lerp_Float(
     A=0
     B=&FN_FrontEndFloatModifier(Gameplay=&Multiply_Float(A=55, B=$System.FN_Contrail_Setup.Dive Value), Front End=0)
     Alpha=1
    )
    Use Spawn Probability=true
   )
   &SpawnRate(
    Spawn Probability=&MakeCustomFloatFromBool(Bool=&SetBoolByFloatComparison(A=$System.Age, B=0.2), True Float=1, False Float=1)
    SpawnRate=&Lerp_Float(A=0, B=500, Alpha=$System.Dive)
    Use Spawn Probability=true
   ) {enabled=false, module=SpawnRate001}
  ]
  particleSpawnScript [
   &InitializeParticle(
    Color Maximum=<rgba 0.25 0.25 0.25 0.05>
    Color Minimum=<rgba 1 1 1 0.025>
    Color Mode=::NewEnumerator2
    Lifetime Max=0.5
    Lifetime Min=0.3
    Lifetime Mode=::NewEnumerator1
    Sprite Rotation Mode=::NewEnumerator2
    Sprite Size=&UniformRangedVector2D(Minimum=<7 100>, Maximum=<1 200>)
    Sprite Size Mode=::NewEnumerator4
    UsePositionOffset=false
    Write Material Random=true
   ) {module=InitializeParticle001}
   &SkeletalMeshLocation(
    Bone / Socket Index=3
    Bone Sampling Mode=::NewEnumerator4
    MaskSpawnGroup=false
    Sampled Orientation Coordinate Space=::Local
    Skeletal Mesh=$System.SkeletalMesh
   )
   &SphereLocation(
    MaskSpawnGroup=false
    Non Uniform Scale=&RotateVectorByQuaternion(Vector To Rotate=<1 1 1>, Quaternion=$Particles.SkeletalMeshLocation.SampledRotation)
    Offset=&FN_FrontEndVector3Modifier(
     Gameplay=&Add_Vector(
      A=&RotateVectorByQuaternion(Vector To Rotate=<100 0 0>, Quaternion=$Particles.SkeletalMeshLocation.SampledRotation)
      B=&TransformVector(Vector=<0 0 -200>, Source Space=::World, Destination Space=::World)
     )
     Front End=<0 0 0>
    )
    Override Seed=false
    Sphere Orientation Axis:Vector3f=&RotateVectorByQuaternion(Vector To Rotate=<100 0 0>, Quaternion=$Particles.SkeletalMeshLocation.SampledRotation)
    Sphere Origin:NiagaraPosition=$Particles.SkeletalMeshLocation.SampledPosition
    Sphere Radius=250
    Surface Only=false
    UseNonUniformScale=true
    UseOffset=true
    UseSphereOrientation=true
   )
   &InheritVelocity(Module.Limit Speed=true)
   &AddVelocity(Velocity=&FN_FrontEndVector3Modifier(Gameplay=<0 0 0>, Front End=<0 -100 100>))
  ]
  particleUpdateScript [
   &ParticleState()
   &SolveForcesAndVelocity(Clamp Velocity=false, Limit Acceleration=false)
   &ScaleSpriteSize(
    Non-Uniform Curve Scale=<1.5 3>
    Non-Uniform Curve Sprite Scale=DataInterface() {propertyValues {
      XCurve {keys [<at -0.0069 0.5 RCIM_Cubic>,<at 0.4918 1 RCIM_Cubic>,<at 1 0.5 RCIM_Cubic>]}
      YCurve {keys [
        {
         arriveTangent=3.1787
         arriveTangentWeight=0.684
         interpMode=RCIM_Cubic
         leaveTangent=3.1787
         leaveTangentWeight=0.684
         tangentMode=RCTM_User
         tangentWeightMode=RCTWM_WeightedBoth
         time=0
         value=0.25
        }
        {
         arriveTangentWeight=0.2354
         interpMode=RCIM_Cubic
         leaveTangentWeight=0.3527
         tangentMode=RCTM_Break
         tangentWeightMode=RCTWM_WeightedBoth
         time=0.4975
         value=1
        }
        <at 1 0.25 RCIM_Cubic>
       ]}
     }}
    Scale Sprite Size Mode=::NewEnumerator3
   )
  ]
  rendererClasses [&NiagaraSpriteRendererProperties]
  renderers [{rendererClass=&NiagaraSpriteRendererProperties}]
 }
}
summary {
 systemName=NS_MuffinLadle_Contrail
 emitters [
  {emitterName=FN_SpeedLines, simTarget=GPUComputeSim, rendererClasses [&NiagaraSpriteRendererProperties]}
  {emitterName=BankingShips, rendererClasses [&NiagaraMeshRendererProperties, &NiagaraSpriteRendererProperties]}
  {bEnabled=false, emitterName=FN_SpeedLines001, simTarget=GPUComputeSim, rendererClasses [&NiagaraSpriteRendererProperties]}
 ]
 userVariables [
  User.bIsFrontEnd {type=NiagaraBool}
  User.bIsFrontEndPreview {type=NiagaraBool}
  User.ParaGlideLeanAlpha {type=NiagaraFloat, defaultValue=0, description=""}
  User.PlayerSpeed {type=NiagaraFloat, defaultValue=0, description=""}
  User.RibbonIDOffset {type=NiagaraInt32}
 ]
}
